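# Backpack character material for WarmShadeWeasel with a reactive emissive scoreboard screen showing an animated tiled logo pattern, kill count and minutes:seconds timer.
# Hashed cells animate the pattern, while digits come from a number atlas via FlipBook, placed by custom UV-remap nodes.
# Elimination, game-start and skydiving timestamps, compared with Time, drive timed blends between displays, with elapsed and healing timer colors.
# Screen emissive is masked and added between the CharacterParent PreFX and PostFX functions.
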Material /BRCosmetics/Accessories/FORT_Backpacks/Backpack_M_MED_WarmShadeWeasel/Materials/M_WarmShadeWeasel_Backpack
graph {
 properties {
  bUseMaterialAttributes=true
  bUsedWithMeshDeformer=true
  bUsedWithSkeletalMesh=true
  pixelDepthOffsetMode=PDOM_Legacy
  shadingModel=MSM_FromMaterialExpression
  usedShadingModels=Default Lit | Subsurface | Clear Coat | Subsurface Profile | Cloth
 }
 nodes [
  n1 = NamedRerouteUsage()
  n2 = Subtract(A=Time(), B=ScalarParameter() {group=Reactivity, parameterName=SkyDivingTime})
  n3 = ScalarParameter() {defaultValue=5, group=Reactivity, parameterName=ReactivityTime_Healing}
  n4 = OneMinus(None=Saturate(None=Divide(A=@n2, B=@n3)))
  n5 = Sign(None=@n4)
  n6 = Divide(B=@n3) {constA=0.5}
  LinearInterpolate(A=NamedRerouteUsage(), B=@n1, Alpha=Multiply(A=@n5, B=SmoothStep(Max=@n6, Value=@n4) {constMax=0.1}))
  n7 = MaterialFunctionCall() {materialFunction=&MF_CharacterParent_2023_PreFX}
  n8 = GetMaterialAttributes(None=@n7.Result) {attributeGetTypes [#1]}
  n9 = Multiply(
   A=NamedRerouteDeclaration(None=ChannelMaskParameter(None=@n7.Tech Art Mask) {group=Board, maskChannel=Blue, parameterName=ScreenArea, defaultValue=<rgba 0 0 1 0>}.R) {name=TechArtMask, nodeColor=<rgba 1 0.8672 0 1>}
   B=NamedRerouteUsage()
  )
  MaterialFunctionCall(
   Material=SetMaterialAttributes(MaterialAttributes=@n8.MaterialAttributes, Emissive Color=Add(A=@n8.Emissive Color, B=@n9)) {attributeSetTypes [#1]}
   M=@n7.M
   Pre Skinned Local Position=@n7.M
   Distance Blend=@n7.M
   Fresnel=@n7.M
  ) {materialFunction=&MF_CharacterParent_2023_PostFX}
  n10 = NamedRerouteDeclaration(None=StaticSwitchParameter(True=TextureCoordinate() {coordinateIndex=1}, False=TextureCoordinate()) {defaultValue=true, group=Board, parameterName=UseUVchannel1}) {name=UV_Org, nodeColor=<rgba 0 0.8125 1 1>}
  n11 = AppendVector(
   A=ScalarParameter() {defaultValue=1, group=Board, parameterName=ScreenTillingX}
   B=ScalarParameter() {defaultValue=1, group=Board, parameterName=ScreenTillingY}
  )
  n12 = ScalarParameter() {group=Board, parameterName=ScreenOffsetX}
  n13 = ScalarParameter() {group=Board, parameterName=ScreenOffsetY}
  n14 = NamedRerouteDeclaration(None=Add(A=Multiply(A=@n10, B=@n11), B=AppendVector(A=@n12, B=@n13))) {name=ScreenUV, nodeColor=<rgba 0 1 0.2734 1>}
  n15 = Multiply(A=@n14, B=ScalarParameter() {defaultValue=32, parameterName=UnitTilling})
  n16 = MaterialFunctionCall!(Seed=Time(), Float, Vec3, Vec2=Ceil(None=@n15)) {materialFunction=&/MF_Hash_WarmShade}
  NamedRerouteDeclaration(None=ComponentMask(None=@n16.Vec2) {r=true}) {name=HashRandom, nodeColor=<rgba 0 0.3672 1 1>}
  n17 = VectorParameter() {group=Reactivity, parameterName=Kill_Center, channelNames {a=ScaleY, b=ScaleX, g=CenterY, r=CenterX}, defaultValue=<rgba 0.5 0.5 1 1>}
  MaterialExpressionAppendVector_5 = AppendVector(A=@n17.CenterX, B=@n17.CenterX)
  MaterialExpressionAppendVector_6 = AppendVector(A=@n17.ScaleX, B=@n17.ScaleX)
  MaterialExpressionNamedRerouteUsage_4 = NamedRerouteUsage()
  n18 = Custom!(UVPoint=@MaterialExpressionAppendVector_5, UVScale=@MaterialExpressionAppendVector_6, OriginalUV=@MaterialExpressionNamedRerouteUsage_4) {
   code="float2 DistUV =OriginalUV-UVPoint;\r\nDistUV = DistUV/UVScale + float2(0.5,0.5);\r\n\r\nreturn DistUV;"
   description=CopyToPointUV
   outputType=CMOT_Float2
   inputs [
    {
     inputName=UVPoint
     input {expression=&self:MaterialExpressionAppendVector_5}
    }
    {
     inputName=UVScale
     input {expression=&self:MaterialExpressionAppendVector_6}
    }
    {
     inputName=OriginalUV
     input {expression=&self:MaterialExpressionNamedRerouteUsage_4}
    }
   ]
  }
  n19 = TextureObjectParameter() {
   group=Board
   parameterName=NumberTex
   samplerType=SAMPLERTYPE_Masks
   texture=/BRCosmetics/Accessories/FORT_Backpacks/Backpack_M_MED_WarmShadeWeasel/Textures/T_CountBoardNumbers
  }
  n20 = TextureSample(UVs=@n18, Tex=@n19) {
   samplerSource=SSM_Wrap_WorldGroupSettings
   samplerType=SAMPLERTYPE_Masks
   texture=&T_CountBoardNumbers
  }
  n21 = Multiply(
   A=Multiply(
    A=@n20.B
    B=MaterialFunctionCall!(UV_Coord=@n18) {materialFunction=&/MF_UV01_Mask_WarmShade}.Mask
   )
   B=VectorParameter() {group=Reactivity, parameterName=Kill_Back, defaultValue=<rgba 0.1479 0.0141 0 1>}.RGB
  )
  n22 = VectorParameter() {group=Reactivity, parameterName=Kill_Front_Dark, defaultValue=<rgba 0.2021 0.1409 0 1>}
  n23 = NamedRerouteDeclaration(None=Clamp(None=ScalarParameter() {group=Reactivity, parameterName=Kill_Number}) {maxDefault=99}) {name=KillNumber, nodeColor=<rgba 1 0.3047 0 1>}
  n24 = Constant() {r=10}
  n25 = Constant() {r=16}
  n26 = Divide(A=Floor(None=Divide(A=@n23, B=@n24)), B=@n25)
  n27 = Divide(A=Constant() {r=9}, B=@n25)
  n28 = Constant() {r=4}
  n29 = VectorParameter() {
   group=Reactivity
   parameterName=Kill_NumberCenter01
   channelNames {a=ScaleY, b=ScaleX, g=CenterY, r=CenterX}
   defaultValue=<rgba 0.5 0.5 1 1>
  }
  MaterialExpressionAppendVector_2 = AppendVector(A=@n29.CenterX, B=@n29.CenterX)
  MaterialExpressionAppendVector_3 = AppendVector(A=@n29.ScaleX, B=@n29.ScaleX)
  MaterialExpressionNamedRerouteUsage_8 = NamedRerouteUsage()
  n30 = Custom(UVPoint=@MaterialExpressionAppendVector_2, UVScale=@MaterialExpressionAppendVector_3, OriginalUV=@MaterialExpressionNamedRerouteUsage_8) {
   code="float2 DistUV =OriginalUV-UVPoint;\r\nDistUV = DistUV/UVScale + float2(0.5,0.5);\r\n\r\nreturn DistUV;"
   description=CopyToPointUV
   outputType=CMOT_Float2
   inputs [
    {
     inputName=UVPoint
     input {expression=&self:MaterialExpressionAppendVector_2}
    }
    {
     inputName=UVScale
     input {expression=&self:MaterialExpressionAppendVector_3}
    }
    {
     inputName=OriginalUV
     input {expression=&self:MaterialExpressionNamedRerouteUsage_8}
    }
   ]
  }
  n31 = TextureSample(
   UVs=MaterialFunctionCall("Animation  Phase (0-1)"=Clamp(None=@n26, Max=@n27), Number of Rows=@n28, Number of Columns=@n28, UVs=@n30) {materialFunction=&FlipBook}.UVs
   Tex=@n19
  ) {
   samplerSource=SSM_Wrap_WorldGroupSettings
   samplerType=SAMPLERTYPE_Masks
   texture=&T_CountBoardNumbers
  }
  n32 = NamedRerouteUsage()
  n33 = Constant() {r=10}
  n34 = Constant() {r=16}
  n35 = Divide(A=Floor(None=Fmod(A=@n32, B=@n33)), B=@n34)
  n36 = Divide(A=Constant() {r=9}, B=@n34)
  n37 = Constant() {r=4}
  MaterialExpressionAdd_8 = Add(A=@MaterialExpressionAppendVector_2, B=AppendVector(A=ScalarParameter() {group=Reactivity, parameterName=Kill_Xoffset}, B=Constant()))
  n38 = Custom(UVPoint=@MaterialExpressionAdd_8, UVScale=@MaterialExpressionAppendVector_3, OriginalUV=@MaterialExpressionNamedRerouteUsage_8) {
   code="float2 DistUV =OriginalUV-UVPoint;\r\nDistUV = DistUV/UVScale + float2(0.5,0.5);\r\n\r\nreturn DistUV;"
   description=CopyToPointUV
   outputType=CMOT_Float2
   inputs [
    {
     inputName=UVPoint
     input {expression=&self:MaterialExpressionAdd_8}
    }
    {
     inputName=UVScale
     input {expression=&self:MaterialExpressionAppendVector_3}
    }
    {
     inputName=OriginalUV
     input {expression=&self:MaterialExpressionNamedRerouteUsage_8}
    }
   ]
  }
  n39 = TextureSample(
   UVs=MaterialFunctionCall("Animation  Phase (0-1)"=Clamp(None=@n35, Max=@n36), Number of Rows=@n37, Number of Columns=@n37, UVs=@n38) {materialFunction=&FlipBook}.UVs
   Tex=NamedRerouteUsage()
  ) {
   samplerSource=SSM_Wrap_WorldGroupSettings
   samplerType=SAMPLERTYPE_Masks
   texture=&T_CountBoardNumbers
  }
  n40 = Add(
   A=Multiply(
    A=@n31.R
    B=MaterialFunctionCall(UV_Coord=@n30) {materialFunction=&MF_UV01_Mask_WarmShade}.Mask
   )
   B=Multiply(
    A=@n39.R
    B=MaterialFunctionCall(UV_Coord=@n38) {materialFunction=&MF_UV01_Mask_WarmShade}.Mask
   )
  )
  n41 = VectorParameter() {group=Reactivity, parameterName=Kill_Front_Bright, defaultValue=<rgba 1 0.6975 0 1>}
  n42 = ScalarParameter() {defaultValue=1, group=Board, parameterName=Kill_NumberIntensity}
  NamedRerouteDeclaration(None=LinearInterpolate(A=LinearInterpolate(A=@n21, B=@n22.RGB, Alpha=@n40), B=Multiply(A=@n41.RGB, B=@n42), Alpha=SmoothStep(Value=@n40) {constMin=0.9})) {name=ScreenKill, nodeColor=<rgba 0 0.3906 1 1>}
  NamedRerouteDeclaration(None=ComponentMask(None=@n16.Vec2) {g=true}) {name=HashRandom_Animate, nodeColor=<rgba 0 0.3672 1 1>}
  n43 = VectorParameter() {group=Board, parameterName=BackColor, defaultValue=<rgba 0.6062 0.1673 0 1>}
  n44 = VectorParameter() {group=Board, parameterName=FrontColor, defaultValue=<rgba 0.7896 0.4811 0 1>}
  n45 = ScalarParameter() {defaultValue=3, group=Board, parameterName=FrontColorIntensity}
  n46 = LinearInterpolate(Alpha=NamedRerouteUsage()) {constB=0.3}
  n47 = Add(A=NamedRerouteUsage(), B=Multiply(A=Time(), B=ScalarParameter() {defaultValue=1, group=Board, parameterName=FlushingSpeed}))
  n48 = TextureObjectParameter() {
   group=Board
   parameterName=MainScreen
   samplerType=SAMPLERTYPE_Masks
   texture=/BRCosmetics/Accessories/FORT_Backpacks/Backpack_M_MED_WarmShadeWeasel/Textures/T_LogoMask
  }
  n49 = TextureSample(UVs=@n15, Tex=@n48) {
   samplerSource=SSM_Wrap_WorldGroupSettings
   samplerType=SAMPLERTYPE_Masks
   texture=&T_LogoMask
  }
  n50 = Multiply(A=LinearInterpolate(A=ScalarParameter() {group=Board, parameterName=MinOpacity}, Alpha=Frac(None=@n47)) {constA=0.2}, B=@n49.G)
  n51 = TextureSample(UVs=@n14, Tex=@n48) {
   samplerSource=SSM_Wrap_WorldGroupSettings
   samplerType=SAMPLERTYPE_Masks
   texture=&T_LogoMask
  }
  n52 = Multiply(
   A=LinearInterpolate(
    A=NamedRerouteDeclaration(None=@n43.RGB) {name=BackColor, nodeColor=<rgba 0.875 1 0 1>}
    B=Multiply(A=@n44.RGB, B=@n45)
    Alpha=SmoothStep(Min=@n46, Value=@n50)
   )
   B=@n51.R
  )
  n53 = Add(A=NamedRerouteUsage(), B=NamedRerouteUsage())
  NamedRerouteDeclaration(None=Multiply(A=@n52, B=Saturate(None=@n53))) {name=HashNoise, nodeColor=<rgba 0.8047 0 1 1>}
  n54 = Time()
  n55 = ScalarParameter() {group=Reactivity, parameterName=ElimationTime}
  n56 = Saturate(None=Divide(A=Subtract(A=@n54, B=@n55), B=ScalarParameter() {defaultValue=5, group=Reactivity, parameterName=ReactivityTime}))
  n57 = OneMinus(None=@n56)
  n58 = Multiply(A=SmoothStep(Value=@n56) {constMax=0.1}, B=SmoothStep(Value=@n57) {constMax=0.1})
  NamedRerouteDeclaration(None=LinearInterpolate(A=NamedRerouteUsage(), B=NamedRerouteUsage(), Alpha=@n58)) {name=Kill_Reactivity, nodeColor=<rgba 1 0 0.8203 1>}
  NamedRerouteDeclaration(None=Power(Base=OneMinus(None=@n58))) {name=Noise_Add, nodeColor=<rgba 0 1 0.2734 1>}
  n59 = LinearInterpolate(
   A=VectorParameter() {group=Reactivity, parameterName=ElapsedTimerColorBackGround, defaultValue=<rgba 0.1479 0.0141 0 1>}.RGB
   B=VectorParameter() {group=Reactivity, parameterName=HealingTimerColorBackGround, defaultValue=<rgba 0 0.0834 0.1479 1>}.RGB
   Alpha=NamedRerouteUsage()
  )
  n60 = VectorParameter() {group=Reactivity, parameterName=Timer_Center, channelNames {a=ScaleY, b=ScaleX, g=CenterY, r=CenterX}, defaultValue=<rgba 0.5 0.5 1 1>}
  MaterialExpressionAppendVector_7 = AppendVector(A=@n60.CenterX, B=@n60.CenterX)
  MaterialExpressionAppendVector_8 = AppendVector(A=@n60.ScaleX, B=@n60.ScaleX)
  MaterialExpressionNamedRerouteUsage_0 = NamedRerouteUsage()
  n61 = Custom(UVPoint=@MaterialExpressionAppendVector_7, UVScale=@MaterialExpressionAppendVector_8, OriginalUV=@MaterialExpressionNamedRerouteUsage_0) {
   code="float2 DistUV =OriginalUV-UVPoint;\r\nDistUV = DistUV/UVScale + float2(0.5,0.5);\r\n\r\nreturn DistUV;"
   description=CopyToPointUV
   outputType=CMOT_Float2
   inputs [
    {
     inputName=UVPoint
     input {expression=&self:MaterialExpressionAppendVector_7}
    }
    {
     inputName=UVScale
     input {expression=&self:MaterialExpressionAppendVector_8}
    }
    {
     inputName=OriginalUV
     input {expression=&self:MaterialExpressionNamedRerouteUsage_0}
    }
   ]
  }
  n62 = TextureSample(UVs=@n61, Tex=NamedRerouteUsage()) {
   samplerSource=SSM_Wrap_WorldGroupSettings
   samplerType=SAMPLERTYPE_Masks
   texture=&T_CountBoardNumbers
  }
  n63 = NamedRerouteDeclaration(None=Multiply(
    A=@n62.G
    B=MaterialFunctionCall(UV_Coord=@n61) {materialFunction=&MF_UV01_Mask_WarmShade}.Mask
   )) {name=TimerBackground, nodeColor=<rgba 1 0 0.9375 1>}
  n64 = Constant() {r=16}
  n65 = Divide(A=NamedRerouteUsage(), B=@n64)
  n66 = Divide(A=Constant() {r=9}, B=@n64)
  n67 = Constant() {r=4}
  n68 = VectorParameter() {group=Reactivity, parameterName=TimerNumber_Center, channelNames {a=ScaleY, b=ScaleX, g=CenterY, r=CenterX}, defaultValue=<rgba 0.5 0.5 1 1>}
  MaterialExpressionAppendVector_9 = AppendVector(A=@n68.CenterX, B=@n68.CenterX)
  MaterialExpressionNamedRerouteDeclaration_19 = NamedRerouteDeclaration(None=AppendVector(A=@n68.ScaleX, B=@n68.ScaleX)) {name=TimerScale, nodeColor=<rgba 1 0.1406 0 1>}
  MaterialExpressionNamedRerouteUsage_17 = NamedRerouteUsage()
  n69 = Custom(
   UVPoint=@MaterialExpressionAppendVector_9
   UVScale=@MaterialExpressionNamedRerouteDeclaration_19
   OriginalUV=@MaterialExpressionNamedRerouteUsage_17
  ) {
   code="float2 DistUV =OriginalUV-UVPoint;\r\nDistUV = DistUV/UVScale + float2(0.5,0.5);\r\n\r\nreturn DistUV;"
   description=CopyToPointUV
   outputType=CMOT_Float2
   inputs [
    {
     inputName=UVPoint
     input {expression=&self:MaterialExpressionAppendVector_9}
    }
    {
     inputName=UVScale
     input {expression=&self:MaterialExpressionNamedRerouteDeclaration_19}
    }
    {
     inputName=OriginalUV
     input {expression=&self:MaterialExpressionNamedRerouteUsage_17}
    }
   ]
  }
  n70 = TextureSample(
   UVs=MaterialFunctionCall("Animation  Phase (0-1)"=Clamp(None=@n65, Max=@n66), Number of Rows=@n67, Number of Columns=@n67, UVs=@n69) {materialFunction=&FlipBook}.UVs
   Tex=NamedRerouteUsage()
  ) {
   samplerSource=SSM_Wrap_WorldGroupSettings
   samplerType=SAMPLERTYPE_Masks
   texture=&T_CountBoardNumbers
  }
  n71 = NamedRerouteDeclaration(None=Multiply(
    A=@n70.R
    B=MaterialFunctionCall(UV_Coord=@n69) {materialFunction=&MF_UV01_Mask_WarmShade}.Mask
   )) {name=TimeMinute_Ten, nodeColor=<rgba 0.6406 1 0 1>}
  n72 = NamedRerouteUsage()
  n73 = Add(A=Add(A=Add(A=@n71, B=@n72), B=NamedRerouteUsage()), B=NamedRerouteUsage())
  n74 = Multiply(
   A=LinearInterpolate(
    A=VectorParameter() {group=Reactivity, parameterName=ElapsedTimerColor, defaultValue=<rgba 1 0.0776 0 1>}.RGB
    B=VectorParameter() {group=Reactivity, parameterName=HealingTimerColor, defaultValue=<rgba 0 0.7532 1 1>}.RGB
    Alpha=NamedRerouteUsage()
   )
   B=ScalarParameter() {defaultValue=1, parameterName=ElapsedTimerIntensity}
  )
  n75 = Multiply(A=Add(A=Multiply(A=@n73) {constB=0.1}, B=Power(Base=@n73) {constExponent=3}), B=@n74)
  n76 = Multiply(A=Multiply(A=@n74) {constB=2}, B=NamedRerouteUsage())
  NamedRerouteDeclaration(None=Add(A=Multiply(A=@n59, B=@n63), B=Add(A=@n75, B=@n76))) {name=ElapsedTimeBoard, nodeColor=<rgba 0 1 0.7422 1>}
  n77 = Constant() {r=16}
  n78 = Divide(A=NamedRerouteUsage(), B=@n77)
  n79 = Divide(A=Constant() {r=9}, B=@n77)
  n80 = Constant() {r=4}
  MaterialExpressionAdd_7 = Add(A=@MaterialExpressionAppendVector_9, B=AppendVector(A=ScalarParameter() {group=Reactivity, parameterName=XOffset_Timer01}, B=Constant()))
  MaterialExpressionNamedRerouteUsage_21 = NamedRerouteUsage()
  MaterialExpressionNamedRerouteUsage_22 = NamedRerouteUsage()
  n81 = Custom(UVPoint=@MaterialExpressionAdd_7, UVScale=@MaterialExpressionNamedRerouteUsage_21, OriginalUV=@MaterialExpressionNamedRerouteUsage_22) {
   code="float2 DistUV =OriginalUV-UVPoint;\r\nDistUV = DistUV/UVScale + float2(0.5,0.5);\r\n\r\nreturn DistUV;"
   description=CopyToPointUV
   outputType=CMOT_Float2
   inputs [
    {
     inputName=UVPoint
     input {expression=&self:MaterialExpressionAdd_7}
    }
    {
     inputName=UVScale
     input {expression=&self:MaterialExpressionNamedRerouteUsage_21}
    }
    {
     inputName=OriginalUV
     input {expression=&self:MaterialExpressionNamedRerouteUsage_22}
    }
   ]
  }
  n82 = TextureSample(
   UVs=MaterialFunctionCall("Animation  Phase (0-1)"=Clamp(None=@n78, Max=@n79), Number of Rows=@n80, Number of Columns=@n80, UVs=@n81) {materialFunction=&FlipBook}.UVs
   Tex=NamedRerouteUsage()
  ) {
   samplerSource=SSM_Wrap_WorldGroupSettings
   samplerType=SAMPLERTYPE_Masks
   texture=&T_CountBoardNumbers
  }
  NamedRerouteDeclaration(None=Multiply(
    A=@n82.R
    B=MaterialFunctionCall(UV_Coord=@n81) {materialFunction=&MF_UV01_Mask_WarmShade}.Mask
   )) {name=TimeMinutes_Single, nodeColor=<rgba 1 0.8672 0 1>}
  NamedRerouteDeclaration(None=@n52) {name=MainPattern, nodeColor=<rgba 0 0.1328 1 1>}
  n83 = Constant() {r=16}
  n84 = Divide(A=NamedRerouteUsage(), B=@n83)
  n85 = Divide(A=Constant() {r=9}, B=@n83)
  n86 = Constant() {r=4}
  MaterialExpressionAdd_11 = Add(A=@MaterialExpressionAdd_7, B=AppendVector(A=ScalarParameter() {group=Reactivity, parameterName=XOffset_Timer02}, B=Constant()))
  MaterialExpressionNamedRerouteUsage_24 = NamedRerouteUsage()
  MaterialExpressionNamedRerouteUsage_25 = NamedRerouteUsage()
  n87 = Custom(UVPoint=@MaterialExpressionAdd_11, UVScale=@MaterialExpressionNamedRerouteUsage_24, OriginalUV=@MaterialExpressionNamedRerouteUsage_25) {
   code="float2 DistUV =OriginalUV-UVPoint;\r\nDistUV = DistUV/UVScale + float2(0.5,0.5);\r\n\r\nreturn DistUV;"
   description=CopyToPointUV
   outputType=CMOT_Float2
   inputs [
    {
     inputName=UVPoint
     input {expression=&self:MaterialExpressionAdd_11}
    }
    {
     inputName=UVScale
     input {expression=&self:MaterialExpressionNamedRerouteUsage_24}
    }
    {
     inputName=OriginalUV
     input {expression=&self:MaterialExpressionNamedRerouteUsage_25}
    }
   ]
  }
  n88 = TextureSample(
   UVs=MaterialFunctionCall("Animation  Phase (0-1)"=Clamp(None=@n84, Max=@n85), Number of Rows=@n86, Number of Columns=@n86, UVs=@n87) {materialFunction=&FlipBook}.UVs
   Tex=NamedRerouteUsage()
  ) {
   samplerSource=SSM_Wrap_WorldGroupSettings
   samplerType=SAMPLERTYPE_Masks
   texture=&T_CountBoardNumbers
  }
  NamedRerouteDeclaration(None=Multiply(
    A=@n88.R
    B=MaterialFunctionCall(UV_Coord=@n87) {materialFunction=&MF_UV01_Mask_WarmShade}.Mask
   )) {name=TimeSecond_Ten, nodeColor=<rgba 0 0.1797 1 1>}
  n89 = Constant() {r=16}
  n90 = Divide(A=NamedRerouteUsage(), B=@n89)
  n91 = Divide(A=Constant() {r=9}, B=@n89)
  n92 = Constant() {r=4}
  MaterialExpressionAdd_13 = Add(A=@MaterialExpressionAdd_11, B=AppendVector(A=ScalarParameter() {group=Reactivity, parameterName=XOffset_Timer03}, B=Constant()))
  MaterialExpressionNamedRerouteUsage_27 = NamedRerouteUsage()
  MaterialExpressionNamedRerouteUsage_28 = NamedRerouteUsage()
  n93 = Custom(UVPoint=@MaterialExpressionAdd_13, UVScale=@MaterialExpressionNamedRerouteUsage_27, OriginalUV=@MaterialExpressionNamedRerouteUsage_28) {
   code="float2 DistUV =OriginalUV-UVPoint;\r\nDistUV = DistUV/UVScale + float2(0.5,0.5);\r\n\r\nreturn DistUV;"
   description=CopyToPointUV
   outputType=CMOT_Float2
   inputs [
    {
     inputName=UVPoint
     input {expression=&self:MaterialExpressionAdd_13}
    }
    {
     inputName=UVScale
     input {expression=&self:MaterialExpressionNamedRerouteUsage_27}
    }
    {
     inputName=OriginalUV
     input {expression=&self:MaterialExpressionNamedRerouteUsage_28}
    }
   ]
  }
  n94 = TextureSample(
   UVs=MaterialFunctionCall("Animation  Phase (0-1)"=Clamp(None=@n90, Max=@n91), Number of Rows=@n92, Number of Columns=@n92, UVs=@n93) {materialFunction=&FlipBook}.UVs
   Tex=NamedRerouteUsage()
  ) {
   samplerSource=SSM_Wrap_WorldGroupSettings
   samplerType=SAMPLERTYPE_Masks
   texture=&T_CountBoardNumbers
  }
  NamedRerouteDeclaration(None=Multiply(
    A=@n94.R
    B=MaterialFunctionCall(UV_Coord=@n93) {materialFunction=&MF_UV01_Mask_WarmShade}.Mask
   )) {name=TimeSecond_Single, nodeColor=<rgba 0.5937 1 0 1>}
  n95 = ScalarParameter() {group=Reactivity, parameterName=SkyDivingActivate}
  n96 = Floor(None=LinearInterpolate(
    A=ScalarParameter() {group=Reactivity, parameterName=TimeNumber}
    B=NamedRerouteUsage()
    Alpha=NamedRerouteDeclaration(None=@n95) {name=SkyDivingActivate, nodeColor=<rgba 1 0 0.0469 1>}
   ))
  n97 = Constant() {r=60}
  n98 = NamedRerouteDeclaration(None=Floor(None=Divide(A=@n96, B=@n97))) {name=RemainingMinutes, nodeColor=<rgba 1 0.1641 0 1>}
  n99 = Constant() {r=10}
  NamedRerouteDeclaration(None=Fmod(A=@n98, B=@n99)) {name=Minutes_Single, nodeColor=<rgba 0 0.1562 1 1>}
  NamedRerouteDeclaration(None=Floor(None=Divide(A=@n98, B=@n99))) {name=Minutes_Ten, nodeColor=<rgba 0.4766 1 0 1>}
  NamedRerouteDeclaration(None=Sign(None=@n57)) {name=ReactivityKillAlpha, nodeColor=<rgba 0.1719 1 0 1>}
  n100 = NamedRerouteDeclaration(None=Fmod(A=@n96, B=Constant() {r=60})) {name=RemainingSeconds, nodeColor=<rgba 0 1 0.4844 1>}
  n101 = Constant() {r=10}
  NamedRerouteDeclaration(None=Fmod(A=@n100, B=@n101)) {name=Seconds_Single, nodeColor=<rgba 0 1 0.5547 1>}
  NamedRerouteDeclaration(None=Floor(None=Divide(A=@n100, B=@n101))) {name=Seconds_Ten, nodeColor=<rgba 1 0.2344 0 1>}
  NamedRerouteDeclaration(None=@n7.Distance Blend) {name=DistanceBlend, nodeColor=<rgba 1 0.3281 0 1>}
  n102 = ComponentMask(None=@n61) {r=true}
  n103 = Multiply(
   A=Step(Y=ScalarParameter() {defaultValue=0.3, group=Reactivity, parameterName=ColuminMin}, X=@n102)
   B=Step(Y=@n102, X=ScalarParameter() {defaultValue=0.7, group=Reactivity, parameterName=ColuminMax})
  )
  NamedRerouteDeclaration(None=Multiply(A=NamedRerouteDeclaration(None=@n103) {name=ColumnMask, nodeColor=<rgba 0.7578 0 1 1>}, B=@n63)) {name=ColumnSignal, nodeColor=<rgba 0 1 0.4375 1>}
  n104 = Fmod(
   A=Subtract(A=Time(), B=ScalarParameter() {defaultValue=-20, group=Reactivity, parameterName=GameStartTime})
   B=ScalarParameter() {defaultValue=70, parameterName=ElapsedLoopTime}
  )
  n105 = ScalarParameter() {defaultValue=12, group=Reactivity, parameterName=ExisitingTime}
  NamedRerouteDeclaration(None=Multiply(A=Step(X=@n104), B=Step(Y=@n104, X=@n105) {constX=15, constY=10})) {name=ElapsedAlpha, nodeColor=<rgba 0 1 0.5781 1>}
  n106 = Divide(A=Constant() {r=0.5}, B=NamedRerouteUsage())
  n107 = Saturate(None=Divide(A=@n104, B=NamedRerouteDeclaration(None=@n105) {name=LifeTimeSwitch, nodeColor=<rgba 0.9922 1 0 1>}))
  n108 = OneMinus(None=@n107)
  NamedRerouteDeclaration(None=Multiply(A=SmoothStep(Max=@n106, Value=@n107) {constMax=0.03}, B=SmoothStep(Max=@n106, Value=@n108) {constMax=0.03})) {name=ElapsedSwitchAlpha, nodeColor=<rgba 1 0.2578 0 1>}
  NamedRerouteDeclaration(None=@n5) {name=ReactivityHealingAlpha, nodeColor=<rgba 0.1719 1 0 1>}
  NamedRerouteDeclaration(None=@n1) {name=SkyDiving_Reactivity, nodeColor=<rgba 1 0 0.8203 1>}
  NamedRerouteDeclaration(None=@n2) {name=SkyTime, nodeColor=<rgba 0.9219 1 0 1>}
  n109 = LinearInterpolate(
   A=NamedRerouteUsage()
   B=LinearInterpolate(A=NamedRerouteUsage(), B=NamedRerouteUsage(), Alpha=NamedRerouteUsage())
   Alpha=NamedRerouteUsage()
  )
  n110 = NamedRerouteUsage()
  n111 = NamedRerouteUsage()
  NamedRerouteDeclaration(None=LinearInterpolate(A=LinearInterpolate(A=@n109, B=@n110, Alpha=@n111), B=NamedRerouteUsage(), Alpha=NamedRerouteUsage())) {name=EM_OUT, nodeColor=<rgba 1 0 0.0703 1>}
  NamedRerouteDeclaration(None=@n19) {name=NumberTex, nodeColor=<rgba 0.6641 1 0 1>}
  ScalarParameter() {group=Reactivity, parameterName=TimeNumber_SkyDiving}
 ]
}
